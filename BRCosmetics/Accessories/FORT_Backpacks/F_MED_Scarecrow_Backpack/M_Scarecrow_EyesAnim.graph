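# Translucent unlit mesh-particle material for the animated glowing eyes of the Scarecrow backpack cosmetic.
# DynamicParameter drives a vertex offset: Blink scales inward displacement along normals, weighted by red or green vertex color (blended by Param4), plus X and Z shifts.
# EyeColor feeds MF_FX_EmissiveAdjustments for the glow.
# An opacity term combines particle alpha, bypassable via UseParticleAlpha, with a TwoSidedSign facing mask.
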Material /BRCosmetics/Accessories/FORT_Backpacks/F_MED_Scarecrow_Backpack/M_Scarecrow_EyesAnim
graph {
 properties {
  bUsedWithMeshParticles=true
  bUsedWithNiagaraMeshParticles=true
  blendMode=BLEND_Translucent
  pixelDepthOffsetMode=PDOM_Legacy
  shadingModel=MSM_Unlit
  usedShadingModels=Unlit
 }
 nodes [
  n1 = DynamicParameter() {paramNames [Blink,Z_Offset,X_Offset,Param4]}
  n2 = Constant3Vector()
  n3 = Constant3Vector() {constant=<rgba 0 0 0.5 0>}
  n4 = VertexColor()
  n5 = Multiply(A=LinearInterpolate(A=Multiply(A=@n4.R) {constB=2}, B=Multiply(A=@n4.G) {constB=2}, Alpha=@n1.Param4), B=VertexNormalWS())
  n6 = LinearInterpolate(Alpha=@n1.Blink) {constA=-0.6, constB=0}
  Add(
   A=LinearInterpolate(A=Constant3Vector(), B=Constant3Vector() {constant=<rgba 0.5 0 0 0>}, Alpha=@n1.X_Offset)
   B=Add(A=LinearInterpolate(A=@n2, B=@n3, Alpha=@n1.Z_Offset), B=Multiply(A=@n5, B=@n6))
  )
  MaterialFunctionCall(Emissive=VectorParameter() {parameterName=EyeColor, defaultValue=<rgba 12.833 0.8775 20 1>}.RGB) {materialFunction=&MF_FX_EmissiveAdjustments}
  n7 = ParticleColor()
  n8 = OneMinus(None=ScalarParameter() {desc="1 = yes, 0 = no", parameterName=UseParticleAlpha})
  Multiply(A=Saturate(None=Add(A=@n7.A, B=@n8)), B=LinearInterpolate(Alpha=TwoSidedSign()))
 ]
}
