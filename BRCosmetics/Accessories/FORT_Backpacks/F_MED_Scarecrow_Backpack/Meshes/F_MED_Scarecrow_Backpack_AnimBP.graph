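# Animation Blueprint for the Scarecrow backpack cosmetic that adds procedural secondary motion to its eye, perch and ribbon bones.
# A single component-space chain nudges the body bone, then runs spring-driven Anim Dynamics on both eyes and, at low alpha, the perch.
# Two three-bone Trail controllers then drive the ribbons, using planar limits attached to the body bone.
# The graph has no animation input or variables, and the event graph has no logic.
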AnimBlueprint /BRCosmetics/Accessories/FORT_Backpacks/F_MED_Scarecrow_Backpack/Meshes/F_MED_Scarecrow_Backpack_AnimBP
target_skeleton=&/F_MED_Scarecrow_Backpack_Skeleton
graphs [
 AnimGraph {nodes [
   n1 = "Transform (Modify) Bone - Bone: body"!(ComponentPose=Local To Component(), Alpha=1) -> (Pose) {
    rotationSpace=BCS_WorldSpace
    translationMode=BMM_Additive
    alphaBoolBlend {blendOption=Linear}
    boneToModify {boneName=body}
    translation {y=1}
   }
   n2 = Anim Dynamics - Bone: dyn_eye_l(ComponentPose=@n1, Alpha=1, ExternalForce="0,0,0") {
    angularDampingOverride=0.9999
    angularSpringConstant=500
    bAngularSpring=true
    bGravityOverrideInSimSpace=true
    bOverrideAngularDamping=true
    bUsePlanarLimit=false
    gravityScale=0.6
    linearSpringConstant=5
    alphaBoolBlend {blendOption=Linear}
    boundBone {boneName=dyn_eye_l}
    componentLinearAccScale=<1 1 1>
    componentLinearVelScale=<1 1 1>
    physicsBodyDefinitions [{
      sphereCollisionRadius=0
      boundBone {boneName=dyn_eye_l}
      boxExtents=<22 1 22>
      constraintSetup {angularTargetAxis=AxisY, twistAxis=AxisY, angularLimitsMax=<10 10 15>, angularLimitsMin=<-10 -10 -5>, angularTarget=<0 0 0>}
      localJointOffset=<-5 -3 0>
     }]
   }
   n3 = Anim Dynamics - Bone: dyn_eye_r(ComponentPose=@n2, Alpha=1, ExternalForce="0,0,0") {
    angularDampingOverride=0.9999
    angularSpringConstant=500
    bAngularSpring=true
    bGravityOverrideInSimSpace=true
    bOverrideAngularDamping=true
    bUsePlanarLimit=false
    gravityScale=0.6
    linearSpringConstant=5
    alphaBoolBlend {blendOption=Linear}
    boundBone {boneName=dyn_eye_r}
    componentLinearAccScale=<1 1 1>
    componentLinearVelScale=<1 1 1>
    physicsBodyDefinitions [{
      sphereCollisionRadius=0
      boundBone {boneName=dyn_eye_r}
      boxExtents=<22 1 22>
      constraintSetup {angularTargetAxis=AxisY, twistAxis=AxisY, angularLimitsMax=<10 10 15>, angularLimitsMin=<-10 -10 -5>, angularTarget=<0 0 0>}
      localJointOffset=<-5 -3 0>
     }]
   }
   n4 = Anim Dynamics - Bone: dyn_perch1(ComponentPose=@n3, Alpha=0.2, ExternalForce="0,0,0") {
    angularDampingOverride=0.99
    angularSpringConstant=500
    bAngularSpring=true
    bGravityOverrideInSimSpace=true
    bOverrideAngularDamping=true
    gravityScale=0.1
    alphaBoolBlend {blendOption=Linear}
    boundBone {boneName=dyn_perch1}
    componentAppliedLinearAccClamp=<1000 1000 1000>
    physicsBodyDefinitions [{
      sphereCollisionRadius=0
      boundBone {boneName=dyn_perch1}
      boxExtents=<1 4 4>
      constraintSetup {angularLimitsMax=<0 0 10>, angularLimitsMin=<0 0 -10>, angularTarget=<0 0 0>}
      localJointOffset=<-3 0 0>
     }]
   }
   n5 = Trail controller - Bone: dyn_ribbonA3(ComponentPose=@n4, Alpha=1) {
    bUsePlanarLimit=true
    chainLength=3
    planarLimits [
     {drivingBone {boneName=body}, planeTransform {location=<0 0 -7>, rotation=<rot 0 -180 -180>, scale=<0.2 0.2 0.2>}}
     {drivingBone {boneName=body}, planeTransform {location=<-5 7 0>, rotation=<rot 0 0 -90>, scale=<0.5 0.5 0.5>}}
    ]
    rotationLimits [{limitMax=<180 180 180>}, {limitMax=<180 180 180>}, {limitMax=<180 180 180>}]
    rotationOffsets [<0 0 0>, <0 0 0>, <0 0 0>]
    trailBone {boneName=dyn_ribbonA3}
   }
   n6 = Trail controller - Bone: dyn_ribbonB3(ComponentPose=@n5, Alpha=1) {
    bUsePlanarLimit=true
    chainLength=3
    planarLimits [
     {drivingBone {boneName=body}, planeTransform {location=<0 0 -7>, rotation=<rot 0 -180 -180>, scale=<0.2 0.2 0.2>}}
     {drivingBone {boneName=body}, planeTransform {location=<-5 7 0>, rotation=<rot 0 0 -90>, scale=<0.5 0.5 0.5>}}
    ]
    rotationLimits [{limitMax=<180 180 180>}, {limitMax=<180 180 180>}, {limitMax=<180 180 180>}]
    rotationOffsets [<0 0 0>, <0 0 0>, <0 0 0>]
    trailBone {boneName=dyn_ribbonB3}
   }
   Output Pose(Result=Component To Local(ComponentPose=@n6)) {layerGroup=None}
  ]}
 EventGraph {nodes [Event BlueprintUpdateAnimation()]}
]
variables []
