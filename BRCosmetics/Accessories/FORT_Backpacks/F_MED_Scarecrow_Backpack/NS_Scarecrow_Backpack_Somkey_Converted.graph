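# Niagara effect for the Scarecrow back bling, converted from Cascade, with smoke, glow and animated eyes.
# Smokey, Glow and GlowStart continuously spawn offset meshes and sprites with color and alpha shifting over life.
# Eyes01 and Eyes02 follow socket FX_eye_L, blinking through Y-scale and dynamic material curves.
# Eyes01 bursts on a loop and writes to System.DeathArray; a system ReadDeath module and DeathSpawn flag chain Eyes02's differently timed animation.
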NiagaraSystem /BRCosmetics/Accessories/FORT_Backpacks/F_MED_Scarecrow_Backpack/NS_Scarecrow_Backpack_Somkey_Converted
dependencies {
 usedDataInterfaces [
  &NiagaraDataInterfaceVectorCurve
  &NiagaraDataInterfaceCurve
  &NiagaraDataInterfaceSkeletalMesh
  &NiagaraDataInterfaceArrayFloat3
  &NiagaraDataInterfaceArrayPosition
 ]
 usedDynamicInputs []
 usedModules [
  &self:SystemSpawnScript.SystemScriptSource.SystemScriptGraph.NiagaraNodeAssignment_0.SetVariables_#1
  &SystemState
  &self:ReadDeath
  &self:SystemSpawnScript.SystemScriptSource.SystemScriptGraph.NiagaraNodeAssignment_1.SetVariables_#2
 ]
}
emitters {
 Eyes01 {
  emitterUpdateScript [
   &EmitterState(Inactive Response=::NewEnumerator1, Life Cycle Mode=::NewEnumerator1, Loop Duration=10, UseLoopDelay=true)
   &SpawnBurst_Instantaneous(Use Spawn Probability=false, UseLoopCountLimit=false)
  ]
  particleSpawnScript [
   /SetVariables(Particles.SkeletonSamplingIndex:NiagaraInt32=/CascadeToNiagaraConverter/NiagaraScripts/DynamicInputScripts/SequentialInt(Use Array Data:NiagaraBool=false, First Number:NiagaraInt32=0, Sequence:NiagaraDataInterfaceArrayInt32=?, Last Number:NiagaraInt32=0)) {isSetParameters=true}
   &InitializeParticle(Lifetime=5, Mesh Scale=<1 0.3 1>, Mesh Scale Mode=::NewEnumerator4, UsePositionOffset=false, Write Material Random=true)
   &InitialMeshOrientation(Mesh Orientation Mode=::NewEnumerator7, Rotation=<0 -0.25 0>, Use Rotation Vector=true)
  ]
  particleUpdateScript [
   &ParticleState()
   &SkeletalMeshLocation(
    Bone / Socket Index=$Particles.SkeletonSamplingIndex
    MaskSpawnGroup=false
    Mesh Sampling Type=::NewEnumerator5
    Orientation Sampling=::NewEnumerator1
    Sampled Position Offset=&SelectVectorFromArray(Direct Array Index=$Particles.SkeletonSamplingIndex, Vector Selection Array=DataInterface() {propertyValues {FloatData [<-2 -4 2>]}})
    Skeletal Mesh=DataInterface() {propertyValues {
      DefaultMesh=null
      ExcludeBoneName=root
      OptionalComponentClass=null
      PreviewMesh=/BRCosmetics/Accessories/FORT_Backpacks/F_MED_Scarecrow_Backpack/Meshes/F_MED_Scarecrow_Backpack
      SkinningMode=SkinOnTheFly
      SoftSourceActor=null
      SourceMode=Default
      UvSetIndex=0
      WholeMeshLOD=-1
      bExcludeBone=false
      bReadDeformedGeometry=true
      bRequireCurrentFrameData=true
      ComponentTags []
      FilteredBones []
      FilteredSockets [FX_eye_L]
      MeshUserParameter {parameter {name=None, typeDefHandle {registeredTypeIndex=1}, varData []}}
      SamplingRegions []
     }}
    Socket Sampling Mode=::NewEnumerator5
   )
   &ScaleMeshSize(Scale Factor=&VectorFromCurve(VectorCurve=DataInterface() {propertyValues {
       XCurve {keys [<at 0 1>,<at 0.22 1 RCIM_Cubic>,<at 0.2336 1 RCIM_Cubic>,<at 0.251 1 RCIM_Cubic>,<at 0.9491 1 RCIM_Cubic>,<at 0.9569 1 RCIM_Cubic>,<at 0.9919 1 RCIM_Cubic>,<at 1 1>]}
       YCurve {keys [
         <at 0 1>
         {arriveTangent=-4.2329, interpMode=RCIM_Cubic, leaveTangent=-4.2329, time=0.22, value=1}
         <at 0.2336 0.0113 RCIM_Cubic>
         {arriveTangent=1.3818, interpMode=RCIM_Cubic, leaveTangent=1.3818, time=0.251, value=1}
         {arriveTangent=-1.3886, interpMode=RCIM_Cubic, leaveTangent=-1.3886, time=0.9491, value=1}
         {arriveTangent=-23.0417, interpMode=RCIM_Cubic, leaveTangent=-23.0417, time=0.9569, value=0.0198}
         {arriveTangent=22.7363, interpMode=RCIM_Cubic, leaveTangent=22.7363, time=0.9919, value=0.0126}
         <at 1 1>
        ]}
       ZCurve {keys [<at 0 1>,<at 0.22 1 RCIM_Cubic>,<at 0.2336 1 RCIM_Cubic>,<at 0.251 1 RCIM_Cubic>,<at 0.9491 1 RCIM_Cubic>,<at 0.9569 1 RCIM_Cubic>,<at 0.9919 1 RCIM_Cubic>,<at 1 1>]}
      }}))
   &Color(Color=&MakeLinearColorFromVectorAndFloat(
     "Vector (RGB)"=&VectorFromCurve(VectorCurve=DataInterface() {propertyValues {XCurve {keys [<at 0 1>,<at 1 1>]}, YCurve {keys [<at 0 1>,<at 1 1>]}, ZCurve {keys [<at 0 1>,<at 1 1>]}}})
     "Float (Alpha)"=&ClampFloat(Float=&FloatFromCurve(FloatCurve=DataInterface() {propertyValues {Curve {keys [{arriveTangent=-1, time=0, value=1}, <at 1 0>]}}}))
    ))
   &self:WriteDeath(Position Array:NiagaraDataInterfaceArrayPosition=$System.DeathArray) {module=WriteDeath}
   &DynamicMaterialParameters(
    Index 0 Param 1=&FloatFromCurve(FloatCurve=DataInterface() {propertyValues {Curve {keys [
         {arriveTangent=0.0008, time=0, value=1}
         {arriveTangent=-4.3773, interpMode=RCIM_Cubic, leaveTangent=-4.3773, time=0.2154, value=1.0008}
         {arriveTangent=1.0116, leaveTangent=1.0116, time=0.2311, value=-0.0116}
         {arriveTangent=0.0843, interpMode=RCIM_Cubic, leaveTangent=0.0843, tangentMode=RCTM_User, time=0.2568, value=1}
         {arriveTangent=-1.4099, interpMode=RCIM_Cubic, leaveTangent=-1.4099, time=0.9367, value=1}
         {arriveTangent=-21.1248, interpMode=RCIM_Cubic, leaveTangent=-21.1248, time=0.961, value=0.0071}
         {arriveTangent=25.4902, interpMode=RCIM_Cubic, leaveTangent=25.4902, time=0.9836, value=0.0091}
         <at 1 1>
        ]}}})
    Index 0 Param 2=&FloatFromCurve(FloatCurve=DataInterface() {propertyValues {Curve {keys [
         {arriveTangent=1.031, time=0.2247, value=-0.0116}
         {arriveTangent=-0.1556, interpMode=RCIM_Cubic, leaveTangent=-0.1556, tangentMode=RCTM_User, time=0.2427, value=1.0194}
         {arriveTangent=-4.2443, interpMode=RCIM_Cubic, leaveTangent=-4.2443, time=0.96, value=0.9829}
         {arriveTangent=-24.6016, interpMode=RCIM_Cubic, leaveTangent=-24.6016, time=0.9748, value=-2.0877}
         <at 1 0>
        ]}}})
    Index 0 Param 3=&FloatFromCurve(FloatCurve=DataInterface() {propertyValues {Curve {keys [
         {arriveTangent=-0.0116, time=0, value=0}
         {arriveTangent=-5.5716, interpMode=RCIM_Cubic, leaveTangent=-5.5716, time=0.2241, value=-0.0116}
         {arriveTangent=-1.8923, interpMode=RCIM_Cubic, leaveTangent=-1.8923, time=0.2505, value=-1.3957}
         {arriveTangent=1.8622, interpMode=RCIM_Cubic, leaveTangent=1.8622, time=0.9585, value=-1.4012}
         <at 1 0>
        ]}}})
    Index 0 Param 4=0.5
    Module.Param0WriteEnabled=true
    Module.Param1WriteEnabled=true
    Module.Param2WriteEnabled=true
    Module.Param3WriteEnabled=true
   )
  ]
  rendererClasses [&NiagaraMeshRendererProperties]
  renderers [{rendererClass=&NiagaraMeshRendererProperties, rendererIndex=0}]
 }
 Eyes02 {
  emitterUpdateScript [
   &EmitterState(Inactive Response=::NewEnumerator1, Life Cycle Mode=::NewEnumerator1, Loop Duration=10, UseLoopDelay=true)
   &SpawnPerFrame(Spawn:NiagaraBool=$System.DeathSpawn, Use Spawn Probability=false)
  ]
  particleSpawnScript [
   /SetVariables(Particles.SkeletonSamplingIndex:NiagaraInt32=&SequentialInt(Use Array Data:NiagaraBool=false, First Number:NiagaraInt32=0, Sequence:NiagaraDataInterfaceArrayInt32=?, Last Number:NiagaraInt32=0)) {isSetParameters=true}
   &InitializeParticle(Lifetime=5, Mesh Scale=<1 0.3 1>, Mesh Scale Mode=::NewEnumerator4, UsePositionOffset=false, Write Material Random=true)
   &InitialMeshOrientation(Mesh Orientation Mode=::NewEnumerator7, Rotation=<0 -0.25 0>, Use Rotation Vector=true)
  ]
  particleUpdateScript [
   &ParticleState()
   &SkeletalMeshLocation(
    Bone / Socket Index=$Particles.SkeletonSamplingIndex
    MaskSpawnGroup=false
    Mesh Sampling Type=::NewEnumerator5
    Orientation Sampling=::NewEnumerator1
    Sampled Position Offset=&SelectVectorFromArray(Direct Array Index=$Particles.SkeletonSamplingIndex, Vector Selection Array=DataInterface() {propertyValues {FloatData [<-2 -4 2>]}})
    Skeletal Mesh=DataInterface() {propertyValues {
      DefaultMesh=null
      ExcludeBoneName=root
      OptionalComponentClass=null
      PreviewMesh=&F_MED_Scarecrow_Backpack
      SkinningMode=SkinOnTheFly
      SoftSourceActor=null
      SourceMode=Default
      UvSetIndex=0
      WholeMeshLOD=-1
      bExcludeBone=false
      bReadDeformedGeometry=true
      bRequireCurrentFrameData=true
      ComponentTags []
      FilteredBones []
      FilteredSockets [FX_eye_L]
      MeshUserParameter {parameter {name=None, typeDefHandle {registeredTypeIndex=1}, varData []}}
      SamplingRegions []
     }}
    Socket Sampling Mode=::NewEnumerator5
   )
   &ScaleMeshSize(Scale Factor=&VectorFromCurve(VectorCurve=DataInterface() {propertyValues {
       XCurve {keys [<at 0 1>,<at 0.0882 1 RCIM_Cubic>,<at 0.1015 1 RCIM_Cubic>,<at 0.1222 1 RCIM_Cubic>,<at 0.4451 1 RCIM_Cubic>,<at 0.4617 1 RCIM_Cubic>,<at 0.475 1 RCIM_Cubic>,<at 0.5208 1 RCIM_Cubic>,<at 0.5441 1 RCIM_Cubic>,<at 0.5533 1 RCIM_Cubic>,<at 0.8193 1 RCIM_Cubic>,<at 0.8371 1 RCIM_Cubic>,<at 0.929 1 RCIM_Cubic>,<at 1 1>]}
       YCurve {keys [
         <at 0 1>
         {arriveTangent=-9.7196, interpMode=RCIM_Cubic, leaveTangent=-9.7196, time=0.0882, value=1}
         <at 0.1015 0.0138 RCIM_Cubic>
         {arriveTangent=2.8697, interpMode=RCIM_Cubic, leaveTangent=2.8697, time=0.1222, value=1}
         {arriveTangent=-2.6648, interpMode=RCIM_Cubic, leaveTangent=-2.6648, time=0.4451, value=1}
         <at 0.4617 0.0953 RCIM_Cubic>
         {arriveTangent=15.3134, interpMode=RCIM_Cubic, leaveTangent=15.3134, time=0.475, value=1}
         {arriveTangent=-13.5228, interpMode=RCIM_Cubic, leaveTangent=-13.5228, time=0.5208, value=1}
         <at 0.5441 0.0663 RCIM_Cubic>
         {arriveTangent=3.3809, interpMode=RCIM_Cubic, leaveTangent=3.3809, time=0.5533, value=1}
         {arriveTangent=-3.4642, interpMode=RCIM_Cubic, leaveTangent=-3.4642, time=0.8193, value=0.997}
         {arriveTangent=-8.9774, interpMode=RCIM_Cubic, leaveTangent=-8.9774, time=0.8371, value=0.0168}
         {arriveTangent=6.0365, interpMode=RCIM_Cubic, leaveTangent=6.0365, time=0.929, value=0.0126}
         <at 1 1>
        ]}
       ZCurve {keys [<at 0 1>,<at 0.0882 1 RCIM_Cubic>,<at 0.1015 1 RCIM_Cubic>,<at 0.1222 1 RCIM_Cubic>,<at 0.4451 1 RCIM_Cubic>,<at 0.4617 1 RCIM_Cubic>,<at 0.475 1 RCIM_Cubic>,<at 0.5208 1 RCIM_Cubic>,<at 0.5441 1 RCIM_Cubic>,<at 0.5533 1 RCIM_Cubic>,<at 0.8193 1 RCIM_Cubic>,<at 0.8371 1 RCIM_Cubic>,<at 0.929 1 RCIM_Cubic>,<at 1 1>]}
      }}))
   &Color(Color=&MakeLinearColorFromVectorAndFloat(
     "Vector (RGB)"=&VectorFromCurve(VectorCurve=DataInterface() {propertyValues {XCurve {keys [<at 0 1>,<at 1 1>]}, YCurve {keys [<at 0 1>,<at 1 1>]}, ZCurve {keys [<at 0 1>,<at 1 1>]}}})
     "Float (Alpha)"=&ClampFloat(Float=&FloatFromCurve(FloatCurve=DataInterface() {propertyValues {Curve {keys [{arriveTangent=-1, time=0, value=1}, <at 1 0>]}}}))
    ))
   &DynamicMaterialParameters(
    Index 0 Param 1=&FloatFromCurve(FloatCurve=DataInterface() {propertyValues {Curve {keys [
         {arriveTangent=0.0124, time=0, value=1}
         {arriveTangent=-10, interpMode=RCIM_Cubic, leaveTangent=-10, time=0.0843, value=1.0124}
         {arriveTangent=0.992, leaveTangent=0.992, time=0.1, value=0}
         {arriveTangent=0.0843, interpMode=RCIM_Cubic, leaveTangent=0.0843, tangentMode=RCTM_User, time=0.1257, value=0.992}
         {arriveTangent=-2.8904, interpMode=RCIM_Cubic, leaveTangent=-2.8904, time=0.4413, value=0.9989}
         {arriveTangent=-0.0015, interpMode=RCIM_Cubic, leaveTangent=-0.0015, time=0.4612, value=0.0222}
         {arriveTangent=19.2728, interpMode=RCIM_Cubic, leaveTangent=19.2728, time=0.4827, value=0.9988}
         {arriveTangent=-15.9032, interpMode=RCIM_Cubic, leaveTangent=-15.9032, time=0.5119, value=0.9988}
         {arriveTangent=-0.0324, interpMode=RCIM_Cubic, leaveTangent=-0.0324, time=0.5441, value=0.0221}
         {arriveTangent=3.6396, interpMode=RCIM_Cubic, leaveTangent=3.6396, time=0.5733, value=0.9968}
         {arriveTangent=-3.9677, interpMode=RCIM_Cubic, leaveTangent=-3.9677, time=0.812, value=0.9971}
         {arriveTangent=-7.8649, interpMode=RCIM_Cubic, leaveTangent=-7.8649, time=0.8177, value=0.0272}
         {arriveTangent=7.1163, interpMode=RCIM_Cubic, leaveTangent=7.1163, time=0.9391, value=-0.0025}
         <at 0.9543 1>
        ]}}})
    Index 0 Param 2=&FloatFromCurve(FloatCurve=DataInterface() {propertyValues {Curve {keys [
         {arriveTangent=-1.2528, time=0.0936, value=0}
         {arriveTangent=-0.1556, interpMode=RCIM_Cubic, leaveTangent=-0.1556, tangentMode=RCTM_User, time=0.1073, value=-1.2528}
         {arriveTangent=5.2827, interpMode=RCIM_Cubic, leaveTangent=5.2827, time=0.5252, value=-1.2868}
         {arriveTangent=-0.6113, interpMode=RCIM_Cubic, leaveTangent=-0.6113, tangentMode=RCTM_User, time=0.5519, value=1.0961}
         {arriveTangent=-2.4464, interpMode=RCIM_Cubic, leaveTangent=-2.4464, time=0.8776, value=1.1254}
         <at 1 0>
        ]}}})
    Index 0 Param 3=&FloatFromCurve(FloatCurve=DataInterface() {propertyValues {Curve {keys [
         <at 0 0>
         {arriveTangent=-13.4812, interpMode=RCIM_Cubic, leaveTangent=-13.4812, time=0.093, value=0}
         {arriveTangent=-4.0168, interpMode=RCIM_Cubic, leaveTangent=-4.0168, time=0.1288, value=-1.7368}
         {arriveTangent=8.0191, interpMode=RCIM_Cubic, leaveTangent=8.0191, time=0.5207, value=-1.7178}
         {arriveTangent=-0.3813, interpMode=RCIM_Cubic, leaveTangent=-0.3813, tangentMode=RCTM_User, time=0.5405, value=1.5645}
         {arriveTangent=-3.4048, interpMode=RCIM_Cubic, leaveTangent=-3.4048, time=0.8778, value=1.5616}
         <at 1 0>
        ]}}})
    Index 0 Param 4=0.5
    Module.Param0WriteEnabled=true
    Module.Param1WriteEnabled=true
    Module.Param2WriteEnabled=true
    Module.Param3WriteEnabled=true
   )
  ]
  rendererClasses [&NiagaraMeshRendererProperties]
  renderers [{rendererClass=&NiagaraMeshRendererProperties, rendererIndex=0}]
 }
 Glow {
  emitterUpdateScript [
   &EmitterState(Inactive Response=::NewEnumerator1, Life Cycle Mode=::NewEnumerator1, Loop Duration=1, UseLoopDelay=true)
   &SpawnRate(SpawnRate=1.3, Use Spawn Probability=false)
  ]
  particleSpawnScript [
   &InitializeParticle(
    Lifetime=&RandomRangeFloat(Minimum=1, Maximum=1.5, Evaluation Type=::NewEnumerator1)
    Sprite Size=<50 75>
    Sprite Size Mode=::NewEnumerator4
    UsePositionOffset=false
    Write Material Random=true
   )
   &AddVelocity(Velocity=<-1 0 0>)
   &Color(Color=&MakeLinearColorFromVectorAndFloat("Vector (RGB)"=<0.8027 0.6452 1>, "Float (Alpha)"=0.2))
   /CascadeToNiagaraConverter/NiagaraScripts/ModuleScripts/AddVectorToPosition(Vector:Vector3f=&TransformVector(Vector=<0 -10 2>, Destination Space=::Simulation))
  ]
  particleUpdateScript [
   &ParticleState()
   &SolveForcesAndVelocity(Clamp Velocity=false, Limit Acceleration=false)
   &ScaleColor(
    Scale Alpha=&FloatFromCurve(FloatCurve=DataInterface() {propertyValues {Curve {keys [<at 0 0>,<at 0.3 1>,<at 1 0>]}}})
    Scale RGB=&VectorFromCurve(VectorCurve=DataInterface() {propertyValues {
       XCurve {keys [<at 0 0.2071>,<at 0.5 0.179>,<at 1 0.3464>]}
       YCurve {keys [<at 0 0.87>,<at 0.5 0.5114>,<at 1 0.1357>]}
       ZCurve {keys [<at 0 1>,<at 0.5 1>,<at 1 1>]}
      }})
    ScaleA=true
    ScaleRGB=true
   )
   &DynamicMaterialParameters(
    Index 0 Param 1=0.8
    Index 0 Param 2=2
    Index 0 Param 3=1
    Index 0 Param 4=1
    Module.Param0WriteEnabled=true
    Module.Param1WriteEnabled=true
    Module.Param2WriteEnabled=true
    Module.Param3WriteEnabled=true
   )
  ]
  rendererClasses [&NiagaraSpriteRendererProperties]
  renderers [{rendererClass=&NiagaraSpriteRendererProperties}]
 }
 GlowStart {
  emitterUpdateScript [
   &EmitterState(Inactive Response=::NewEnumerator1, Life Cycle Mode=::NewEnumerator1, Loop Duration=1, UseLoopDelay=true)
   &SpawnRate(SpawnRate=1.3, Use Spawn Probability=false)
  ]
  particleSpawnScript [
   &InitializeParticle(Lifetime=1.5, Mesh Scale=<0.3 0.3 0.3>, Mesh Scale Mode=::NewEnumerator4, UsePositionOffset=false, Write Material Random=true)
   &Color(Color=&MakeLinearColorFromVectorAndFloat("Vector (RGB)"=<0.3698 0.3698 0.3698>, "Float (Alpha)"=0.1))
   &AddVectorToPosition(Vector:Vector3f=&TransformVector(Vector=<0 -10 2>, Destination Space=::Simulation))
  ]
  particleUpdateScript [
   &ParticleState()
   &ScaleMeshSize(Scale Factor=&VectorFromFloat(Value=&FloatFromCurve(FloatCurve=DataInterface() {propertyValues {Curve {keys [<at 0 1>,<at 1 1.25>]}}})))
   &ScaleColor(
    Scale Alpha=&FloatFromCurve(FloatCurve=DataInterface() {propertyValues {Curve {keys [<at 0 0 RCIM_Cubic>,<at 0.3064 0.7588 RCIM_Cubic>,<at 1 0 RCIM_Cubic>]}}})
    Scale RGB=&VectorFromCurve(VectorCurve=DataInterface() {propertyValues {
       XCurve {keys [<at 0.3 0.0647>,<at 0.7 0.0363>,<at 1 0 RCIM_Cubic>]}
       YCurve {keys [<at 0.3 0.0449>,<at 0.7 0.0006>,<at 1 0 RCIM_Cubic>]}
       ZCurve {keys [<at 0.3 0.3906>,<at 0.7 0.151>,<at 1 0 RCIM_Cubic>]}
      }})
    ScaleA=true
    ScaleRGB=true
   )
   &DynamicMaterialParameters(
    Index 0 Param 1=0.2
    Index 0 Param 2=2
    Index 0 Param 3=8
    Index 0 Param 4=1.25
    Module.Param0WriteEnabled=true
    Module.Param1WriteEnabled=true
    Module.Param2WriteEnabled=true
    Module.Param3WriteEnabled=true
   )
  ]
  rendererClasses [&NiagaraMeshRendererProperties]
  renderers [{rendererClass=&NiagaraMeshRendererProperties, rendererIndex=0}]
 }
 Smokey {
  emitterUpdateScript [
   &EmitterState(Life Cycle Mode=::NewEnumerator1, Loop Duration=1, UseLoopDelay=true)
   &SpawnRate(SpawnRate=4, Use Spawn Probability=false)
  ]
  particleSpawnScript [
   &InitializeParticle(
    Lifetime=&RandomRangeFloat(Minimum=1.5, Maximum=1, Evaluation Type=::NewEnumerator1)
    Mesh Scale=&RandomRangeVector(Minimum=<0.15 0.08 0.1>, Maximum=<0.4 0.05 0.1>, Evaluation Type=::NewEnumerator1)
    Mesh Scale Mode=::NewEnumerator4
    UsePositionOffset=false
    Write Material Random=true
   )
   &AddVelocity(
    Cone Angle=&RandomRangeFloat(Minimum=0, Maximum=0.2, Evaluation Type=::NewEnumerator1)
    Cone Angle Mode:ENiagara_AngleInput=::NewEnumerator1
    Cone Axis=<0 -1 0>
    Module.UseConeFalloff=false
    Module.UseVelDistribution=false
    Override Seed=false
    Velocity Mode=::NewEnumerator2
    Velocity Speed=&RandomRangeFloat(Minimum=3, Maximum=10, Evaluation Type=::NewEnumerator1)
   )
   &Color(Color=&MakeLinearColorFromVectorAndFloat(
     "Vector (RGB)"=&RandomRangeVector(Minimum=<0 0 0>, Maximum=<1 1 1>, Evaluation Type=::NewEnumerator1)
     "Float (Alpha)"=30
    ))
   &DynamicMaterialParameters(
    Index 0 Param 2=&RandomRangeFloat(Minimum=30, Maximum=50, Evaluation Type=::NewEnumerator1)
    Module.Param0WriteEnabled=false
    Module.Param1WriteEnabled=true
    Module.Param2WriteEnabled=false
    Module.Param3WriteEnabled=false
   ) {module=DynamicMaterialParameters001}
   &AddVectorToPosition(Vector:Vector3f=&TransformVector(Vector=<0 -10 2>, Destination Space=::Simulation))
  ]
  particleUpdateScript [
   &ParticleState()
   &Drag(Use Linear Drag=true, Use Rotational Drag=true)
   &SolveForcesAndVelocity(Clamp Velocity=false, Limit Acceleration=false)
   &ScaleColor(
    Scale Alpha=&FloatFromCurve(FloatCurve=DataInterface() {propertyValues {Curve {keys [
         {arriveTangent=0.2972, interpMode=RCIM_Cubic, leaveTangent=0.2972, tangentMode=RCTM_User, time=0, value=-0.0092}
         {arriveTangent=0.0092, interpMode=RCIM_Cubic, leaveTangent=0.0092, time=0.271, value=1.5228}
         <at 1 0 RCIM_Cubic>
        ]}}})
    Scale RGB=&VectorFromCurve(VectorCurve=DataInterface() {propertyValues {
       XCurve {keys [<at 0.0015 1 RCIM_Cubic>,<at 0.5 1 RCIM_Cubic>,<at 0.5619 1 RCIM_Cubic>,<at 1 1 RCIM_Cubic>]}
       YCurve {keys [
         <at 0.0015 22.0935 RCIM_Cubic>
         {arriveTangent=-39.1079, interpMode=RCIM_Cubic, leaveTangent=-39.1079, time=0.5, value=5}
         {arriveTangent=-10, interpMode=RCIM_Cubic, leaveTangent=-10, time=0.5619, value=0.1782}
         <at 1 0 RCIM_Cubic>
        ]}
       ZCurve {keys [
         <at 0.0015 0 RCIM_Cubic>
         {arriveTangent=8.9225, interpMode=RCIM_Cubic, leaveTangent=8.9225, time=0.5, value=5}
         {arriveTangent=490, interpMode=RCIM_Cubic, leaveTangent=490, time=0.5619, value=5}
         <at 1 250 RCIM_Cubic>
        ]}
      }})
    ScaleA=true
    ScaleRGB=true
   )
   &DynamicMaterialParameters(
    Index 0 Param 1=&FloatFromCurve(FloatCurve=DataInterface() {propertyValues {Curve {keys [<at 0 5>,<at 1 0>]}}})
    Index 0 Param 3=&FloatFromCurve(FloatCurve=DataInterface() {propertyValues {Curve {keys [<at 0 3>,<at 1 12>]}}})
    Index 0 Param 4=&FloatFromCurve(FloatCurve=DataInterface() {propertyValues {Curve {keys [
         {arriveTangent=-0.0169, interpMode=RCIM_Cubic, leaveTangent=-0.0169, tangentMode=RCTM_User, time=0.0123, value=0.25}
         {arriveTangent=-1.0347, interpMode=RCIM_Cubic, leaveTangent=-1.0347, time=0.8212, value=0.2237}
         {arriveTangent=-6.8435, interpMode=RCIM_Cubic, leaveTangent=-6.8435, time=0.9123, value=-0.6812}
         <at 1 -1 RCIM_Cubic>
        ]}}})
    Module.Param0WriteEnabled=true
    Module.Param1WriteEnabled=false
    Module.Param2WriteEnabled=true
    Module.Param3WriteEnabled=true
   )
  ]
  rendererClasses [&NiagaraMeshRendererProperties]
  renderers [{rendererClass=&NiagaraMeshRendererProperties, rendererIndex=0}]
 }
}
summary {
 systemName=NS_Scarecrow_Backpack_Somkey_Converted
 emitters [
  {emitterName=Smokey, rendererClasses [&NiagaraMeshRendererProperties]}
  {emitterName=Glow, rendererClasses [&NiagaraSpriteRendererProperties]}
  {emitterName=GlowStart, rendererClasses [&NiagaraMeshRendererProperties]}
  {emitterName=Eyes01, rendererClasses [&NiagaraMeshRendererProperties]}
  {emitterName=Eyes02, rendererClasses [&NiagaraMeshRendererProperties]}
 ]
 userVariables []
}
